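# Control Rig for the F_MED_BrightLarva female player hair that applies corrective offsets to hair bones during targeting and crouching.
# A DeltaRotation function turns one Euler axis of the head's swing/twist rotation relative to head_null into a clamped weight.
# With isTargeting, that weight is interpolated and curve-shaped to drive one Modify Transforms pose.
# With isCrouching, it is multiplied by remapped pelvis height to drive a second.
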ControlRigBlueprint /BRCosmetics/Characters/Player/Female/Medium/Bodies/F_MED_BrightLarva/Meshes/Parts/F_MED_BrightLarvaHair/F_MED_BrightLarvaHair_FaceAcc_CtrlRig
graphs [
 New Function_ContainedGraph {nodes [
   n1 = RigVMFunctionEntryNode()
   RigVMFunction_MathBoolToInteger = &RigVMFunction_MathBoolToInteger(Value=@n1.onlyTwist)
   RigVMFunction_MathQuaternionSwingTwist = &RigVMFunction_MathQuaternionSwingTwist(
    twistAxis=@n1.twistAxis
    Input=&RigUnit_GetRelativeTransformForItem(
     bChildInitial=False
     bParentInitial=False
     child=(Type=Bone,Name="None")
     parent=(Type=Bone,Name="None")
     child.Name=@n1.child
     parent.Name=@n1.parent
     parent.Type=@n1.parentType
    ).RelativeTransform.Rotation
   )
   RigVMFunction_MathQuaternionToEuler = &RigVMFunction_MathQuaternionToEuler(
    Value=RigVMDispatchNode(
     Values=((),())
     Index=@RigVMFunction_MathBoolToInteger.Result
     Values.0=@RigVMFunction_MathQuaternionSwingTwist.Swing
     Values.1=@RigVMFunction_MathQuaternionSwingTwist.Twist
    ) {template="DISPATCH_RigVMDispatch_SelectInt32(in Index,in Values,out Result)"}.Result
    RotationOrder=@n1.Rotation Order
   )
   > DISPATCH_RigVMDispatch_Print = RigVMDispatchNode(ScreenColor=(R=1,G=1,B=1,A=1), ScreenDuration=1.5, Enabled=@n1.debugRotation, Value=@RigVMFunction_MathQuaternionToEuler.Result) {template="DISPATCH_RigVMDispatch_Print(in Prefix,in Value,in Enabled,in ScreenDuration,in ScreenColor)"}
   Clamp = &RigVMFunction_MathIntClamp(Value=@n1.primAxis)
   DISPATCH_RigVMDispatch_SelectInt32 = RigVMDispatchNode(
    Values=(0,0,0)
    Index=@Clamp.Result
    Values.0=@RigVMFunction_MathQuaternionToEuler.Result.X
    Values.1=@RigVMFunction_MathQuaternionToEuler.Result.Y
    Values.2=@RigVMFunction_MathQuaternionToEuler.Result.Z
   ) {template="DISPATCH_RigVMDispatch_SelectInt32(in Index,in Values,out Result)"}
   Remap = &RigVMFunction_MathDoubleRemap(
    TargetMaximum=1
    TargetMinimum=0
    bClamp=False
    Value=@DISPATCH_RigVMDispatch_SelectInt32.Result
    SourceMaximum=@n1.max
    SourceMinimum=@n1.min
   )
   Multiply = &RigVMFunction_MathDoubleMul(A=&RigVMFunction_MathDoubleClamp(Value=@Remap.Result).Result, B=@n1.premult)
   > RigVMFunctionReturnNode(outputWeight=@Multiply.Result)
  ]}
 RigVMFunctionLibrary {nodes [DeltaRotation_1 = RigVMCollapseNode(ExecuteContext=(), Rotation Order=ZYX, premult=1, twistAxis=(X=1,Y=0,Z=0))]}
 RigVMModel {nodes [
   RigUnit_BeginExecution = &RigUnit_BeginExecution()
   DeltaRotation_6 = RigVMFunctionReferenceNode(
    Rotation Order=ZYX
    child=head
    debugRotation=false
    max=-25
    min=0
    parent=head_null
    parentType=Null
    premult=1
    primAxis=2
    twistAxis=(X=1,Y=0,Z=0)
    ExecuteContext=@RigUnit_BeginExecution.ExecutePin
   )
   RigVMFunction_MathBoolToFloat = &RigVMFunction_MathBoolToFloat(Value=RigVMVariableNode() {variable=isTargeting}.Value)
   Multiply = &RigVMFunction_MathDoubleMul(B=@RigVMFunction_MathBoolToFloat.Result, A=@DeltaRotation_6.outputWeight)
   AlphaInterp = &RigVMFunction_AlphaInterp(InterpSpeedDecreasing=10, InterpSpeedIncreasing=10, Scale=1, bClampResult=false, bMapRange=false, Value=@Multiply.Result)
   RigVMFunction_AnimEvalRichCurve = &RigVMFunction_AnimEvalRichCurve(
    Curve="(EditorCurveData=(Keys=((InterpMode=RCIM_Cubic,TangentMode=RCTM_User,TangentWeightMode=RCTWM_WeightedNone,Time=0,Value=0,ArriveTangent=2.4061,ArriveTangentWeight=0,LeaveTangent=2.4061,LeaveTangentWeight=0),(InterpMode=RCIM_Cubic,TangentMode=RCTM_Auto,Tange ...(truncated)"
    SourceMinimum=0
    TargetMinimum=0
    Value=@AlphaInterp.Result
   )
   > &RigUnit_ModifyTransforms(
    ItemToModify="((Item=(Type=Bone,Name=\"base_hair_1_r\"),Transform=(Rotation=(X=0.0048,Y=0.0425,Z=-0.1118,W=-0.9928),Translation=(X=0.2715,Y=0.9583,Z=5.4504),Scale3D=(X=1,Y=1,Z=1))),(Item=(Type=Bone,Name=\"base_hair_0_r\"),Transform=(Rotation=(X=0,Y=0,Z=0,W=-1),Translation=( ...(truncated)"
    WeightMinimum=0
    Weight=@RigVMFunction_AnimEvalRichCurve.Result
   )
   Remap = &RigVMFunction_MathDoubleRemap(
    SourceMaximum=45
    SourceMinimum=89
    TargetMaximum=1
    TargetMinimum=0
    Value=&RigUnit_GetRelativeTransformForItem(Child=(Type=Bone,Name="pelvis"), Parent=(Type=Bone,Name="root"), bChildInitial=False, bParentInitial=False).RelativeTransform.Translation.Z
   )
   n1 = &RigVMFunction_MathBoolToFloat(Value=RigVMVariableNode() {variable=isCrouching}.Value)
   > &RigUnit_ModifyTransforms(
    ItemToModify="((Item=(Type=Bone,Name=\"base_hair_1_r\"),Transform=(Rotation=(X=0.0545,Y=-0.4008,Z=-0.0323,W=0.914),Translation=(X=-0.5664,Y=0.8596,Z=0.2763),Scale3D=(X=1,Y=1,Z=1))),(Item=(Type=Bone,Name=\"base_hair_0_r\"),Transform=(Rotation=(X=0,Y=0,Z=0,W=1),Translation=(X ...(truncated)"
    WeightMinimum=0
    Weight=&RigVMFunction_MathDoubleMul(B=&RigVMFunction_MathDoubleMul(A=@Remap.Result, B=@n1.Result).Result, A=@DeltaRotation_6.outputWeight).Result
   )
  ]}
]
hierarchy [
 root {loc=<0 0 0>, rot=<0 0 0 1>}
 pelvis {parent=root, loc=<0 -0.5301 98.0571>, rot=<0.0127 0.707 -0.0127 -0.707>}
 spine_01 {parent=pelvis, loc=<6.1208 -0.9328 0>, rot=<0 0 0.034 -0.9994>}
 spine_02 {parent=spine_01, loc=<6.5997 -0.2095 0>, rot=<0 0 0.0238 -0.9997>}
 spine_03 {parent=spine_02, loc=<6.4914 -0.1237 0>, rot=<0 0 0.0149 -0.9999>}
 spine_04 {parent=spine_03, loc=<6.9598 0.0794 0>, rot=<0 0 -0.0132 -0.9999>}
 spine_05 {parent=spine_04, loc=<6.9987 0 0>, rot=<0 0 -0.0015 -1>}
 neck_01 {parent=spine_05, loc=<11.3138 0.1529 0>, rot=<0 0 -0.2003 -0.9797>}
 neck_02 {parent=neck_01, loc=<4.5627 0 0>, rot=<0 0 0.0504 -0.9987>}
 head {parent=neck_02, loc=<5.3914 0 0>, rot=<0 0 0.115 -0.9934>}
 base_hair_0_l {parent=head, loc=<2.5097 2.8216 -8.0721>, rot=<0.0645 -0.7042 0.0587 -0.7047>}
 dyn_hair_0_l {parent=base_hair_0_l, loc=<0 0 0>, rot=<0 0 0 1>}
 base_hair_1_l {parent=head, loc=<2.5679 -4.2235 -6.8298>, rot=<0.2733 -0.6522 0.2679 -0.6544>}
 dyn_hair_1_l {parent=base_hair_1_l, loc=<0 0 0>, rot=<0 0 0 1>}
 base_hair_0_c {parent=head, loc=<2.5965 -7.6878 0>, rot=<0.5021 -0.4979 0.4979 -0.5021>}
 dyn_hair_0_c {parent=base_hair_0_c, loc=<0 0 0>, rot=<0 0 0 1>}
 base_hair_0_r {parent=head, loc=<2.5097 2.8216 6.4594>, rot=<0.7047 -0.0587 0.7042 -0.0645>}
 dyn_hair_0_r {parent=base_hair_0_r, loc=<0 0 0>, rot=<0 0 0 1>}
 base_hair_1_r {parent=head, loc=<2.568 -4.2235 5.2171>, rot=<0.6544 -0.2679 0.6522 -0.2733>}
 dyn_hair_1_r {parent=base_hair_1_r, loc=<0 0 0>, rot=<0 0 0 1>}
 head_null {parent=neck_02, type=Null, loc=<5.3914 0 0>, rot=<0 0 0.115 -0.9934>}
 clavicle_l {parent=spine_05, loc=<5.4831 -0.1053 -3.2946>, rot=<0.0465 -0.7289 -0.009 -0.6829>}
 upperarm_l {parent=clavicle_l, loc=<9.8658 -1.2136 -0.4105>, rot=<0.0164 -0.3245 -0.0269 -0.9454>}
 upperarm_twist_01_l {parent=upperarm_l, loc=<7.2135 -0.1134 0>, rot=<0 0 0 1>}
 deform_upperarm_up_base_l {parent=upperarm_twist_01_l, loc=<-8.891 0.4825 3.9517>, rot=<0 0 0 -1>}
 base_arm_puff_l {parent=deform_upperarm_up_base_l, loc=<1.6775 -0.3691 -3.9517>, rot=<0 0.7071 0 -0.7071>}
 dyn_arm_puff_l {parent=base_arm_puff_l, loc=<0 0 0>, rot=<0 0 0 1>}
 clavicle_r {parent=spine_05, loc=<5.4831 -0.1053 3.2946>, rot=<0.7289 0.0465 -0.6829 0.009>}
 upperarm_r {parent=clavicle_r, loc=<-9.8658 1.2136 0.4105>, rot=<0.0164 -0.3245 -0.0269 -0.9454>}
 upperarm_twist_01_r {parent=upperarm_r, loc=<-7.2135 0.1134 0>, rot=<0 0 0 -1>}
 deform_upperarm_up_base_r {parent=upperarm_twist_01_r, loc=<8.891 -0.4825 -3.9517>, rot=<0 0 0 -1>}
 base_arm_puff_r {parent=deform_upperarm_up_base_r, loc=<-1.6773 0.3691 3.9513>, rot=<0 -0.7071 0 -0.7071>}
 dyn_arm_puff_r {parent=base_arm_puff_r, loc=<0 0 0>, rot=<0 0 0 -1>}
]
variables [
 isCrouching {type=bool, default=False}
 isTargeting {type=bool, default=False}
]
